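# Physics asset for the StaminaCat medium male player body cosmetic, covering only the lower body and a center loincloth.
# Capsules wrap the pelvis, thighs, calves and inner and outer knee deform bones, with two capsules shaping each thigh.
# The loincloth is a five-box dyn_ bone chain hung from the pelvis with tight angular limits.
# Hip and knee joints swing and twist freely, while knee deform bodies are tightly constrained to the calves.
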PhysicsAsset /BRCosmetics/Characters/Player/Male/Medium/Bodies/M_MED_StaminaCat/Meshes/M_MED_StaminaCat_Physics
bodies {
 calf_l [{shapeType=Capsule, length=39.0802, radius=5.2484, shapeName=calf_l_capsule, center=<-16.9016 0 0>, rotation=<rot 90 0 0>}]
 calf_r [{shapeType=Capsule, length=39.0802, radius=5.2484, shapeName=calf_l_capsule, center=<16.9016 0 0>, rotation=<rot 90 0 180>}]
 deform_knee_in_l [{shapeType=Capsule, length=0.38, radius=9.947, shapeName=deform_knee_in_l_capsule, center=<0 -7.6327 0>}]
 deform_knee_in_r [{shapeType=Capsule, length=0.38, radius=9.947, shapeName=deform_knee_in_r_capsule, center=<0 8.2832 0>}]
 deform_knee_out_l [{shapeType=Capsule, length=0.38, radius=9.947, shapeName=deform_knee_out_l_capsule, center=<0 5.6744 0>}]
 deform_knee_out_r [{shapeType=Capsule, length=0.38, radius=11.3793, shapeName=deform_knee_out_r_capsule, center=<0 -4.8054 0>}]
 dyn_loinCloth_c_01 [{shapeType=Box, extentX=5.5859, extentY=14.2952, extentZ=2.6724, shapeName=dyn_loinCloth_c_01_box, center=<3.664 0 0>}]
 dyn_loinCloth_c_02 [{shapeType=Box, extentX=8.0782, extentY=14.2952, extentZ=2.5556, shapeName=dyn_loinCloth_c_02_box, center=<5.8987 0 0>}]
 dyn_loinCloth_c_03 [{shapeType=Box, extentX=10.7609, extentY=14.2952, extentZ=2.3324, shapeName=dyn_loinCloth_c_03_box, center=<3.646 0 -0.3627>}]
 dyn_loinCloth_c_04 [{shapeType=Box, extentX=10.7609, extentY=14.2952, extentZ=2.3324, shapeName=dyn_loinCloth_c_04_box, center=<3.6463 0 -0.3599>}]
 dyn_loinCloth_c_05 [{shapeType=Box, extentX=9.5417, extentY=14.2952, extentZ=2.3324, shapeName=dyn_loinCloth_c_05_box, center=<2.4714 0 0>}]
 pelvis [{shapeType=Capsule, length=6.531, radius=9.0364, shapeName=pelvis_capsule, center=<7.1214 0 0>}]
 thigh_l [
  {shapeType=Capsule, length=7.2737, radius=13.9392, shapeName=thigh_l_capsule, center=<-39.1432 0 -2.9714>, rotation=<rot 90 0 0>}
  {
   shapeType=Capsule
   length=29.7785
   radius=8.9952
   shapeName=thigh_l_capsule
   center=<-9.0107 2.1312 -0.669>
   rotation=<rot 89.4269 179.9951 -175.0523>
  }
 ]
 thigh_r [
  {shapeType=Capsule, length=7.2737, radius=13.9392, shapeName=thigh_l_capsule, center=<39.1432 0 2.9714>, rotation=<rot 90 0 180>}
  {
   shapeType=Capsule
   length=29.7785
   radius=8.9952
   shapeName=thigh_l_capsule
   center=<9.0107 -2.1312 0.669>
   rotation=<rot 89.4269 179.9951 4.9477>
  }
 ]
}
constraints [
 {
  bone1Name=calf_l
  bone2Name=thigh_l
  swing1LimitDegrees=45
  swing1Motion=Free
  swing2LimitDegrees=45
  swing2Motion=Free
  twistLimitDegrees=45
  twistMotion=Free
 }
 {
  bone1Name=calf_r
  bone2Name=thigh_r
  swing1LimitDegrees=45
  swing1Motion=Free
  swing2LimitDegrees=45
  swing2Motion=Free
  twistLimitDegrees=45
  twistMotion=Free
 }
 {bone1Name=dyn_loinCloth_c_02, bone2Name=dyn_loinCloth_c_01, swing1LimitDegrees=20, swing2LimitDegrees=20, twistLimitDegrees=20}
 {bone1Name=dyn_loinCloth_c_03, bone2Name=dyn_loinCloth_c_02, swing1LimitDegrees=20, swing2LimitDegrees=20, twistLimitDegrees=20}
 {bone1Name=dyn_loinCloth_c_04, bone2Name=dyn_loinCloth_c_03, swing1LimitDegrees=20, swing2LimitDegrees=20, twistLimitDegrees=20}
 {bone1Name=dyn_loinCloth_c_05, bone2Name=dyn_loinCloth_c_04, swing1LimitDegrees=20, swing2LimitDegrees=20, twistLimitDegrees=20}
 {
  bone1Name=thigh_l
  bone2Name=pelvis
  swing1LimitDegrees=45
  swing1Motion=Free
  swing2LimitDegrees=45
  swing2Motion=Free
  twistLimitDegrees=45
  twistMotion=Free
 }
 {bone1Name=dyn_loinCloth_c_01, bone2Name=pelvis, swing1LimitDegrees=30, swing2LimitDegrees=30, twistLimitDegrees=0}
 {
  bone1Name=thigh_r
  bone2Name=pelvis
  swing1LimitDegrees=45
  swing1Motion=Free
  swing2LimitDegrees=45
  swing2Motion=Free
  twistLimitDegrees=45
  twistMotion=Free
 }
 {bone1Name=deform_knee_in_l, bone2Name=calf_l, swing1LimitDegrees=20, swing2LimitDegrees=20, twistLimitDegrees=0}
 {bone1Name=deform_knee_out_l, bone2Name=calf_l, swing1LimitDegrees=20, swing2LimitDegrees=20, twistLimitDegrees=0}
 {bone1Name=deform_knee_in_r, bone2Name=calf_r, swing1LimitDegrees=20, swing2LimitDegrees=20, twistLimitDegrees=0}
 {bone1Name=deform_knee_out_r, bone2Name=calf_r, swing1LimitDegrees=20, swing2LimitDegrees=20, twistLimitDegrees=0}
]
