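# Animation Blueprint for the StaminaCat player body, handling loincloth adjustment and secondary physics.
# Copies the parent mesh pose, runs it through a deform Control Rig, adds pitch to three chained center loincloth bones, then applies pelvis-relative RigidBody simulation.
# Each update, sprinting or tactical sprinting enables the interpolated loincloth rotation, zeroes component motion scales and halves physics alpha; otherwise these revert.
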AnimBlueprint /BRCosmetics/Characters/Player/Male/Medium/Bodies/M_MED_StaminaCat/Meshes/M_MED_StaminaCat_AnimBP
target_skeleton=&/M_MED_StaminaCat_Skeleton
graphs [
 AnimGraph {nodes [
   "Transform (Modify) Bone"(Alpha=1, Translation="0,0,0", Rotation="0,0,0", Scale="1,1,1") {alphaBoolBlend {blendOption=Linear}, boneToModify=~}
   n1 = Control Rig(Source=Copy Pose From Mesh() {bUseAttachedParent=true}) {
    alphaBoolBlend {blendOption=Linear}
    controlRigAssetReference {blueprintRigClass=&/M_MED_DeformRig_StaminaCatMaster_ControlRig.M_MED_DeformRig_StaminaCatMaster_ControlRig_C}
   }
   n2 = Get isSpintTactRunning() -> (isSpintTactRunning:double)
   n3 = "Transform (Modify) Bone - Bone: dyn_loinCloth_c_01"(ComponentPose=Local To Component(LocalPose=@n1), Alpha=@n2, Translation="0,0,0", Rotation="40,0,0", Scale="1,1,1") {
    rotationMode=BMM_Additive
    rotationSpace=BCS_BoneSpace
    alphaBoolBlend {blendOption=Linear}
    alphaScaleBiasClamp {bInterpResult=true, interpSpeedDecreasing=6, interpSpeedIncreasing=6}
    boneToModify {boneName=dyn_loinCloth_c_01}
    rotation {pitch=40}
   }
   n4 = "Transform (Modify) Bone - Bone: dyn_loinCloth_c_02"(ComponentPose=@n3, Alpha=@n2, Translation="0,0,0", Rotation="30,0,0", Scale="1,1,1") {
    rotationMode=BMM_Additive
    rotationSpace=BCS_BoneSpace
    alphaBoolBlend {blendOption=Linear}
    alphaScaleBiasClamp {bInterpResult=true, interpSpeedDecreasing=6, interpSpeedIncreasing=6}
    boneToModify {boneName=dyn_loinCloth_c_02}
    rotation {pitch=30}
   }
   n5 = "Transform (Modify) Bone - Bone: dyn_loinCloth_c_03"(ComponentPose=@n4, Alpha=@n2, Translation="0,0,0", Rotation="20,0,0", Scale="1,1,1") {
    rotationMode=BMM_Additive
    rotationSpace=BCS_BoneSpace
    alphaBoolBlend {blendOption=Linear}
    alphaScaleBiasClamp {bInterpResult=true, interpSpeedDecreasing=6, interpSpeedIncreasing=6}
    boneToModify {boneName=dyn_loinCloth_c_03}
    rotation {pitch=20}
   }
   n6 = Get Acc Scale vals() -> (Acc Scale vals:Vector)
   n7 = RigidBody!(
    ComponentPose=@n5
    Alpha=Get rigidBodyAlpha() -> (rigidBodyAlpha:double)
    ExternalForce
    ComponentLinearAccScale:Vector=@n6
    ComponentLinearVelScale:Vector=@n6
   ) -> (Pose) {
    bClampLinearTranslationLimitToRefPose=true
    bOverrideWorldGravity=true
    overridePhysicsAsset=&/M_MED_StaminaCat_Physics
    simulationSpace=BaseBoneSpace
    alphaBoolBlend {blendOption=Linear}
    baseBoneRef {boneName=pelvis}
    componentAppliedLinearAccClamp=<6000 6000 6000>
    componentLinearAccScale=<0.2 0.2 0.2>
    componentLinearVelScale=<0.2 0.2 0.2>
    overrideWorldGravity {z=-980}
   }
   Output Pose(Result=Component To Local(ComponentPose=@n7)) {layerGroup=None}
  ]}
 EventGraph {nodes [
   Event BlueprintUpdateAnimation()
   > n1 = Branch(Condition=OR Boolean(A=Get bIsSprinting(), B=Get bIsTacticalSprinting()))
   > Set Acc Scale vals(Acc Scale vals:Vector="0, 0, 0") -> (Output_Get:Vector)
   > Set isSpintTactRunning(isSpintTactRunning:double=1) -> (Output_Get:double)
   > Set rigidBodyAlpha(rigidBodyAlpha:double=0.5) -> (Output_Get:double)
   Set Acc Scale vals(execute=@n1.else, Acc Scale vals="0.2,0.2,0.2")
   > Set isSpintTactRunning(isSpintTactRunning=0)
   > Set rigidBodyAlpha(rigidBodyAlpha=1)
  ]}
]
variables [
 Acc Scale vals {type=Vector}
 isSpintTactRunning {type=number}
 rigidBodyAlpha {type=number}
]
